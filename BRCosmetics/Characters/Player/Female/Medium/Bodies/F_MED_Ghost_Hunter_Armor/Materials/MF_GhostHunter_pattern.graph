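# Material function that layers an animated, color-cycling Ghost Hunter armor pattern onto input material attributes through a channel-selectable mask.
# World-aligned noise on pre-skinned position, panned and distorted by second noise, drives a scrolling curve-atlas gradient with Fresnel darkening.
# The same distorted coordinates drive flattened swirl normals and bubble-texture emissive, scaled by time of day.
# The Desaturation and bounds-normalized position branches connect to nothing.
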MaterialFunction /BRCosmetics/Characters/Player/Female/Medium/Bodies/F_MED_Ghost_Hunter_Armor/Materials/MF_GhostHunter_pattern
graph {nodes [
  n1 = TextureObjectParameter() {
   group=GhostHunter
   parameterName=GH_WP_Noise_01
   samplerType=SAMPLERTYPE_Masks
   texture=/Game/Global/Textures/Default/Noise/T_WrapEdgeWear_Noise
  }
  n2 = VectorParameter() {group=GhostHunter, parameterName=GH_WP_NoiseTiling, defaultValue=<rgba 0.5 0.5 0.5 1>}
  n3 = Time()
  n4 = VectorParameter() {group=GhostHunter, parameterName=GH_WP_PanSpeedAndDirection, defaultValue=<rgba 0 0.1 0.1 1>}
  n5 = Divide(A=VertexInterpolator(VS=PreSkinnedPosition()).PS, B=ScalarParameter() {defaultValue=256, group=GhostHunter, parameterName=GH_WP_GlobalScale})
  n6 = TextureObjectParameter() {
   group=GhostHunter
   parameterName=GH_WP_DistortionNoise
   samplerType=SAMPLERTYPE_Masks
   texture=&T_WrapEdgeWear_Noise
  }
  n7 = VectorParameter() {group=GhostHunter, parameterName=GH_WP_DistortionPanSpeedAndDirection, defaultValue=<rgba 0 -0.05 -0.05 1>}
  n8 = Multiply(A=Time(), B=@n7.RGB)
  n9 = VertexInterpolator(VS=PreSkinnedNormal())
  n10 = Multiply(A=MaterialFunctionCall(
    TextureObject=@n6
    TextureSize=VectorParameter() {group=GhostHunter, parameterName=GH_WP_DistortionNoiseTiling, defaultValue=<rgba 1 1 1 1>}.RGB
    WorldPosition=Add(A=@n5, B=@n8)
    World Space Normal=@n9.PS
   ) {materialFunction=/Game/Weapons/WeaponSkins/Wraps/MaterialFunctions/WorldAlignedTexture}.XYZ Texture) {constB=2}
  n11 = Add(
   A=Add(A=Multiply(A=@n3, B=@n4.RGB), B=@n5)
   B=Multiply(A=Subtract(A=@n10), B=ScalarParameter() {defaultValue=0.02, group=GhostHunter, parameterName=GH_WP_DistortionStrength})
  )
  n12 = Multiply(A=MaterialFunctionCall(TextureObject=@n1, TextureSize=@n2.RGB, WorldPosition=@n11, World Space Normal=@n9.PS) {materialFunction=&WorldAlignedTexture}.XYZ Texture) {constB=2}
  n13 = ChannelMaskParameter(None=Multiply(A=Subtract(A=@n12), B=ScalarParameter() {defaultValue=0.1, group=GhostHunter, parameterName=GH_WP_GradientDistortionStrength})) {group=GhostHunter, maskChannel=Blue, parameterName=GH_GradientDirection, defaultValue=<rgba 0 0 1 0>}
  n14 = Multiply(A=@n13.R) {constB=1.5}
  n15 = Multiply(A=Time(), B=ScalarParameter() {defaultValue=0.01, desc=Speed of color cycle, group=GhostHunter, parameterName=GH_GradientPanSpeed}) {constB=0.05}
  n16 = Multiply(A=Frac(None=Subtract(A=@n14, B=@n15)), B=ScalarParameter() {defaultValue=4, group=GhostHunter, parameterName=GH_GradientDivisions}) {constB=4}
  n17 = PreSkinnedPosition()
  n18 = Multiply(A=ComponentMask(None=VertexInterpolator(VS=@n17).PS) {b=true}) {constB=-0.003}
  n19 = CurveAtlasRowParameter(CurveTime=Frac(None=Add(A=@n16, B=@n18))) {
   atlas=&AthenaCurveAtlas01
   curve=/Game/Characters/Player/Female/Medium/Bodies/F_MED_Ghost_Hunter_Armor/Materials/T_GhostHunter_Colors_01
   defaultValue=298
   group=GhostHunter
   parameterName=GH_ColorCurve
  }
  Desaturation(
   None=Multiply(A=@n19, B=ScalarParameter() {defaultValue=1, group=GhostHunter, parameterName=GH_mult}) {constB=0.3}
   Fraction=ScalarParameter() {group=GhostHunter, parameterName=GH_saturation}
  )
  n20 = PreSkinnedPosition()
  Divide(
   A=FunctionInput(Preview=VertexInterpolator(VS=@n20).PS) {bUsePreviewValueAsDefault=true, inputName=Local Position}
   B=PreSkinnedLocalBounds().Max
  )
  n21 = GetMaterialAttributes(None=FunctionInput() {inputName=Material, inputType=FunctionInput_MaterialAttributes}) {attributeGetTypes [#1, #2]}
  n22 = SetMaterialAttributes!(
   MaterialAttributes=@n21.MaterialAttributes
   EmissiveColor=Add(A=@n21.EmissiveColor, B=NamedRerouteUsage())
   Specular=ScalarParameter() {defaultValue=1, group=GhostHunter, parameterName=GH_Specular}
   Roughness=ScalarParameter() {defaultValue=0.15, group=GhostHunter, parameterName=GH_Roughness, sliderMin=1}
   BaseColor=NamedRerouteUsage()
   Normal=NamedRerouteUsage()
   Metallic=ScalarParameter() {group=GhostHunter, parameterName=GH_MEtallic}
  ) {attributeSetTypes [
    #1
    #3
    #4
    #5
    #2
    #6
   ]}
  n23 = FunctionInput(Preview=Constant()) {bUsePreviewValueAsDefault=true, inputName=Mask}
  FunctionOutput(None=BlendMaterialAttributes(A=@n21.MaterialAttributes, B=@n22, Alpha=ChannelMaskParameter(None=@n23) {group=GhostHunter, parameterName=GH_MaskChannel}.R))
  n24 = Multiply(A=Fresnel() {baseReflectFraction=0, exponent=2}) {constB=2}
  NamedRerouteDeclaration(None=LinearInterpolate(A=@n19, B=Multiply(A=@n19) {constB=0.25}, Alpha=Saturate(None=@n24))) {name=Color, nodeColor=<rgba 0.1012 0.1344 1 1>}
  n25 = MaterialFunctionCall(
   TextureObject=TextureObject() {samplerType=SAMPLERTYPE_Normal, texture=/Game/Global/Textures/Default/Noise/T_SwirlNoise_N}
   TextureSize=@n2.RGB
   WorldPosition=Multiply(A=@n11, B=Constant3Vector() {constant=<rgba 2 2 1 0>}) {constB=2}
   World Space Normal=@n9.PS
  ) {materialFunction=&WorldAlignedTexture}
  NamedRerouteDeclaration(None=MaterialFunctionCall(
    Normal=MaterialFunctionCall(BaseNormal=@n21.Normal, AdditionalNormal=@n25.XYZ Texture) {materialFunction=&BlendAngleCorrectedNormals}.Result
    Flatness=ScalarParameter() {defaultValue=0.9, parameterName=GH_NormalFlatness}
   ) {materialFunction=&FlattenNormal}.Result) {name=Normal, nodeColor=<rgba 1 0 0.4687 1>}
  n26 = TextureObject() {
   samplerType=SAMPLERTYPE_LinearGrayscale
   texture=/Game/Characters/Player/Female/Medium/Bodies/F_MED_Ghost_Hunter_Armor/Textures/T_F_MED_Ghost_Hunter_Bubbles
  }
  n27 = MaterialFunctionCall(TextureObject=@n26, TextureSize=@n2.RGB, WorldPosition=@n11, World Space Normal=@n9.PS) {materialFunction=&WorldAlignedTexture}
  n28 = NamedRerouteUsage()
  n29 = Constant() {r=0.66}
  n30 = Constant() {r=0.8}
  n31 = Constant() {r=0.5}
  NamedRerouteDeclaration(None=Multiply(
    A=Multiply(A=@n27.XYZ Texture, B=@n28)
    B=MaterialFunctionCall(Value_Morning=@n29, Value_Day=@n30, Value_Evening=@n29, Value_Night=@n31) {materialFunction=&Scalar_Time_of_Day}.Result
   )) {name=Emissive, nodeColor=<rgba 0 0.9766 1 1>}
  NamedRerouteDeclaration(None=@n19) {name=colorForEmissive, nodeColor=<rgba 0.2656 1 0 1>}
  VertexNormalWS()
 ]}
