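# Material function for the Ghost Hunter armor cosmetic that layers skin effects onto incoming material attributes.
# One SkinFX_Mask texture supplies selectable thin-film and cloak channels, whose sum drives the thin-film alpha, and feeds the metals LUT as RGBA.
# The material passes through thin film, metals LUT and emissive fresnel, then a masked Ghost Hunter pattern function produces the output.
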MaterialFunction /BRCosmetics/Characters/Player/Female/Medium/Bodies/F_MED_Ghost_Hunter_Armor/Materials/MF_GhostHunter
graph {nodes [
  n1 = TextureSample(Tex=NamedRerouteDeclaration(None=TextureObjectParameter() {group=Skin FX, parameterName=SkinFX_Mask}) {name=SkinFX_Mask, nodeColor=<rgba 0.9842 0 1 1>})
  n2 = StaticComponentMaskParameter(None=@n1.RGB) {defaultR=true, group=Skin FX, parameterName=ThinFilm_Channel}
  n3 = StaticComponentMaskParameter(None=@n1.RGB) {defaultR=true, group=Skin FX, parameterName=cloak_Channel}
  n4 = MaterialFunctionCall(
   Mask=@n1.RGBA
   Materials=MaterialFunctionCall!(
    Alpha=Add(A=@n2, B=@n3)
    Material Attributes=FunctionInput() {inputName=Material, inputType=FunctionInput_MaterialAttributes, sortPriority=-20}
   ) {materialFunction=/Game/Weapons/WeaponSkins/MaterialFunction/ThinFilmFunction}.Result
  ) {materialFunction=&MF_MetalsLUT}
  FunctionOutput(None=MaterialFunctionCall!(
    Mask=NamedRerouteUsage()
    Local Position
    Material=MaterialFunctionCall!(In=@n4.Result) {materialFunction=/Game/Weapons/WeaponSkins/MaterialFunction/MF_EmissiveFresnel}.Result
   ) {materialFunction=&/MF_GhostHunter_pattern}.Result)
  NamedRerouteDeclaration(None=@n3) {name=CloakMask, nodeColor=<rgba 1 0.2344 0 1>}
 ]}
