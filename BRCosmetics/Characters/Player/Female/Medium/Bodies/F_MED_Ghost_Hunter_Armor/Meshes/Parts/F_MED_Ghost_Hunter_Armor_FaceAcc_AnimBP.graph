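# Animation Blueprint for the face accessory part of the female Ghost Hunter Armor player cosmetic, which also drives a position parameter in its materials.
# The AnimGraph copies the pose from the attached parent mesh.
# On each animation update, the EventGraph checks the part's skeletal mesh, waits a zero-duration delay, then writes the spine_05 component-space position to the "Cube Location" material vector parameter.
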AnimBlueprint /BRCosmetics/Characters/Player/Female/Medium/Bodies/F_MED_Ghost_Hunter_Armor/Meshes/Parts/F_MED_Ghost_Hunter_Armor_FaceAcc_AnimBP
target_skeleton=&/F_MED_Ghost_Hunter_Armor_FaceAcc_Skeleton
graphs [
 AnimGraph {nodes [Output Pose(Result=Copy Pose From Mesh() {bUseAttachedParent=true}) {layerGroup=None}]}
 EventGraph {nodes [
   Event BlueprintUpdateAnimation()
   n1 = GetPartSkeletalMeshComponent()
   > Is Valid(InputObject=@n1)
   > Delay(Duration=0)
   > SetVectorParameterValueOnMaterials!(
    execute
    self:Mesh Component Object Reference=@n1
    ParameterName:Name=Cube Location
    ParameterValue:Vector=TransformLocation!(T:"Transform (by ref)"=GetSocketTransform(self=@n1, InSocketName=spine_05, TransformSpace=RTS_Component), Location:Vector="0, 0, 0") -> (ReturnValue:Vector)
   ) -> (then)
  ]}
]
variables []
